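; Heist display case actor that shows a treasure item players can steal, or that drops it if destroyed before looting.
; Once playlist data is ready, the server assigns a hotfix-gated manual override or asks the game state's item manager by group tag, waiting if needed.
; The treasure's RepNotify async-loads and centers its skeletal or static mesh and sets marker details by name-matching DataTable rows.
; The interaction prompt shows name and rarity.
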(event EventBeginPlay
  (bind _returnvalue (Fortnite|Athena|GameState|GetGameStateAthena))
  (bind _returnvalue_2 (Actor|GetComponentbyClass _returnvalue "/HeistBasesGameplay/Items/Environmental/DisplayCase/GSC_HeistDisplayCaseItemManager.GSC_HeistDisplayCaseItemManager_C"))
  ;[0 1 0] Beginplay. Bind On Interact, pick our treasure, start cosmetic rotation
  (|CallParentFunction)
  (Class|FortAsyncActionCurrentPlaylistReady|PlaylistDataReadyAsync
    (:OnReady
      (Networking|SwitchHasAuthority
        (:Authority
          (if (and (Utilities|IsValid (Variables|Default|GetTreasureManualOverride) :!class /Script/Engine.KismetSystemLibrary) (ScalableFloat|GetValueasBool (Variables|Default|GetFNMHotfixActive)))
            (CallFunction|ServerSetupTreasure (Variables|Default|GetTreasureManualOverride))
            (else
              (Utilities|IsValid _returnvalue_2
                (:Is%20Valid
                  (bind _output_get (Variables|Default|SetMythic%20Spawn%20Manager _returnvalue_2))
                  (if (Class|FortGameStateComponentHeistDisplayCaseItemManager|IsDataReady _output_get)
                    (bind _returnvalue_3 (Class|FortGameStateComponentHeistDisplayCaseItemManager|GetDisplayCaseItem (Variables|Default|GetMythic%20Spawn%20Manager) (Variables|Default|GetDisplay%20Case%20Group%20Tag)))
                    (Utilities|IsValid _returnvalue_3
                      (:Is%20Valid
                        (CallFunction|ServerSetupTreasure _returnvalue_3)))
                    (else
                      (EventDispatchers|BindEventtoOnDisplayCaseItemManagerReady (Variables|Default|GetMythic%20Spawn%20Manager) (delegate Custom|Display%20Case%20Item%20Manager%20Ready)))))))))))))

(event Custom|Display%20Case%20Item%20Manager%20Ready
  ;[0 1 0] Beginplay. Bind On Interact, pick our treasure, start cosmetic rotation
  (bind _returnvalue (Class|FortGameStateComponentHeistDisplayCaseItemManager|GetDisplayCaseItem (Variables|Default|GetMythic%20Spawn%20Manager) (Variables|Default|GetDisplay%20Case%20Group%20Tag)))
  (Utilities|IsValid _returnvalue
    (:Is%20Valid
      (CallFunction|ServerSetupTreasure _returnvalue))))

(event Custom|Server_GiveReward
  (params (InteractingPawn :type Fort%20Pawn%20Object%20Reference))
  ;[0 0 1] Grant the item, set looted/interactable states
  (Networking|SwitchHasAuthority
    (:Authority
      (Variables|Default|SetLooted true)
      (Variables|Default|SetCanInteract false)
      (Inventory|GiveItemtoInventoryOwner (Fortnite|GetFortPlayerControllerfromActor InteractingPawn) (Variables|Default|GetAssigned%20Treasure)))))

(event Custom|Server_SetupTreasure
  (params (Treasure :type Fort%20World%20Item%20Definition%20Object%20Reference))
  ;[0 1 0] Assign treasure on server and inform clients.
  (Networking|SwitchHasAuthority
    (:Authority
      (Variables|Default|SetAssigned%20Treasure Treasure))))

(event Custom|Client_MeshSetup
  (bind _treasureskeletalmesh (Variables|Default|GetTreasureSkeletalMesh))
  (bind _treasurestaticmesh (Variables|Default|GetTreasureStaticMesh))
  ;[0.5 0 1] Setup skeletal or static mesh on clients
  (Networking|CanExecuteCosmeticEvents
    (:True
      (if (not (Variables|Default|GetLooted))
        (bind _itemdata (Class|ItemDefinitionBase|GetItemData (Variables|Default|GetAssigned%20Treasure))
          (:Found
            (if (Utilities|IsValidSoftObjectReference (.PickupSkeletalMesh (Utilities|Struct|BreakFortnitePickupItemData _itemdata)))
              (bind _object (Utilities|AsyncLoadAsset (.PickupSkeletalMesh (Utilities|Struct|BreakFortnitePickupItemData _itemdata)))
                (:Completed
                  (bind _asskeletal_mesh (Utilities|Casting|CastToSkeletalMesh _object))
                  (Components|SkeletalMesh|SetSkeletalMeshAsset _treasureskeletalmesh _asskeletal_mesh)
                  (Rendering|SetVisibility _treasureskeletalmesh true false)
                  (Rendering|SetVisibility (Variables|Default|GetTreasureStaticMesh) false false)
                  (bind _offset (CallFunction|CenterMesh _treasureskeletalmesh))
                  (Default|CallMeshSetupComplete _offset)))
              (else
                (bind _object_2 (Utilities|AsyncLoadAsset (Utilities|Struct|BreakFortnitePickupItemData _itemdata))
                  (:Completed
                    (bind _asstatic_mesh (Utilities|Casting|CastToStaticMesh _object_2))
                    (Components|StaticMesh|SetStaticMesh _treasurestaticmesh _asstatic_mesh)
                    (Rendering|SetVisibility _treasurestaticmesh true false)
                    (Rendering|SetVisibility (Variables|Default|GetTreasureSkeletalMesh) false false)
                    (bind _offset_2 (CallFunction|CenterMesh _treasurestaticmesh))
                    (Default|CallMeshSetupComplete _offset_2)))))))))))

(event Custom|OnRep_Looted
  ;[0.5 0 1] Hide meshes once looted
  (Networking|CanExecuteCosmeticEvents
    (:True
      (Rendering|SetVisibility (Variables|Default|GetTreasureSkeletalMesh) false false))))

(event EventOnDeathServer
  (params (Damage :type Float%20%28single-precision%29) (DamageTags :type Gameplay%20Tag%20Container%20Structure%20%28by%20ref%29) (Momentum :type Vector :default "0, 0, 0") (HitInfo :type Hit%20Result%20Structure%20%28by%20ref%29) (InstigatedBy :type Controller%20Object%20Reference) (DamageCauser :type Actor%20Object%20Reference) (EffectContext :type Gameplay%20Effect%20Context%20Handle%20Structure))
  ;[0 0 1] Drop the item if the case is destroyed
  (|CallParentFunction Damage DamageTags Momentum HitInfo InstigatedBy DamageCauser EffectContext)
  (Networking|SwitchHasAuthority
    (:Authority
      (if (not (Variables|Default|GetLooted))
        (Inventory|SpawnPickupInWorld (Variables|Default|GetAssigned%20Treasure) 1 (Transformation|GetWorldLocation (Variables|Default|GetTreasureLocation)) "0.000000,0.000000,1.000000" 0 true true false -1 "Tossed" "LootDrop" 0 false)))))

(fn BlueprintCanInteract
  (params (InteractingPawn :type Fort%20Pawn%20Object%20Reference) (InteractionBeingAttempted :type EInteractionBeingAttempted%20Enum :default "FirstInteraction") (InteractionType :type TInteractionType%20Enum :default "IT_NoInteraction"))
  (return (Variables|Default|GetCanInteract)))

(fn BlueprintGetInteractionString
  (params (InteractingPawn :type Fort%20Pawn%20Object%20Reference) (InteractionBeingAttempted :type EInteractionBeingAttempted%20Enum :default "FirstInteraction"))
  (bind _assigned_treasure (Variables|Default|GetAssigned%20Treasure))
  (bind _returnvalue (Item|GetRarity _assigned_treasure))
  (Utilities|IsValid _assigned_treasure
    (:Is%20Valid
      (bind _returnvalue_2 (Fortnite|GetRarityData _returnvalue))
      (return (Utilities|Text|FormatText "NSLOCTEXT(\"[B308295C76F99453FF2B45C9A16598C3]\", \"7379A28644A7802DC16AAA9EC2983053\", \"<Failure_Description>Steal</>\\r\\n{Name}\\r\\n{Rarity}\")" (ItemDefinition|GetDisplayName (Variables|Default|GetAssigned%20Treasure)) (Utilities|Text|FormatText (Utilities|Select :Index _returnvalue :Common "INVTEXT(\"<Rarity_Common>({0})</>\")" :Uncommon "INVTEXT(\"<Rarity_Uncommon>({0})</>\")" :Rare "INVTEXT(\"<Rarity_Rare>({0})</>\")" :Epic "INVTEXT(\"<Rarity_Epic>({0})</>\")" :Legendary "INVTEXT(\"<Rarity_Legendary>({0})</>\")" :Mythic "INVTEXT(\"<Rarity_Mythic>({0})</>\")") (Utilities|Struct|BreakFortRarityItemData _returnvalue_2)))))))

(fn CenterMesh
  (params (InMesh :type Scene%20Component%20Object%20Reference))
  (bind _returnvalue (Transformation|GetWorldLocation InMesh))
  (bind _origin (Collision|GetComponentBounds InMesh))
  (bind _returnvalue_2 (- _returnvalue _origin))
  (bind _returnvalue_3 (* _returnvalue_2 0.5))
  ;[0.5 0 1] Center the mesh nicely
  (Transformation|SetWorldLocation InMesh (+ (Transformation|GetWorldLocation (Variables|Default|GetTreasureLocation)) _returnvalue_3) false true)
  (return _returnvalue_3))

(fn GetMarkerDetailsId
  (return "None"))

(fn OnRep_AssignedTreasure
  (CallFunction|ClientMeshSetup)
  (CallFunction|SetupTreasureMarkerDetails))

(fn SetupTreasureMarkerDetails
  (bind _outrownames (DataTable|GetDataTableRowNames (Variables|Default|GetMarkerDetailsDataTable)))
  (bind _array_element (Utilities|Array|ForEachLoopwithBreak _outrownames)
    (:LoopBody
      (if (Utilities|String|Contains (Utilities|String|ToString(Name) _array_element) (Utilities|GetObjectName (Variables|Default|GetAssigned%20Treasure)) false true)
        (bind _returnvalue (Utilities|GetDataTableRow (Variables|Default|GetMarkerDetailsDataTable) _array_element)
          (:then
            (Variables|Marking|SetMarkerDisplay _returnvalue))
          (:RowNotFound
            _))))))

(fn ConstructionScript)
