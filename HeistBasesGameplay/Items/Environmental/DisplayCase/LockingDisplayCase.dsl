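; Heist base display case Blueprint holding a glowing treasure that players smash open for a reward.
; Interaction applies a GA-granting effect; FinishProcess then rewards the pawn, plays shatter feedback and broadcasts a base lockdown request.
; On lockdown, unlooted cases become uninteractable, raise a barricade, then retract their pedestal via timeline-driven replicated position.
; OnRep Looted swaps to broken glass, and a hotfix flag can start cases underground and disabled.
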(event EventLoottableTreasureSetup)

(event EventBeginPlay
  ;[0.2 1 0.2] Set up mesh materials according to rarity
  (|CallParentFunction)
  (Default|BindEventtoMeshSetupComplete (delegate Custom|MythicMaterialSetup)))

(event Custom|MythicMaterialSetup
  (params (Offset :type Vector))
  ;[0.2 1 0.2] Set up mesh materials according to rarity
  (Utilities|FlowControl|Sequence
    (:then_0
      (Rendering|Material|SetVectorParameterValue (Class|BTechArtGlobalLibrary|GetDynamicMaterialInstance (Variables|Default|GetPedestal)) "SFX_Emissive_Tint(Expand)_3" "(R=10.000000,G=5.000000,B=1.000001,A=1.000000)" :!class /Script/Engine.MaterialInstanceDynamic))
    (:then_1
      (Rendering|SetVisibility (Variables|Default|GetBG) true true)
      (Transformation|AddWorldOffset (Variables|Default|GetBG) Offset false false))
    (:then_2
      (CallFunction|SetMaterialOverlay (Variables|Default|GetTreasureSkeletalMesh))
      (CallFunction|SetMaterialOverlay (Variables|Default|GetTreasureStaticMesh)))))

(event EventEndPlay
  (params (EndPlayReason :type EEndPlayReason%20Enum :default "Destroyed"))
  ;[1 0.2 0] Endplay binding cleanup
  (Default|UnbindEventfromMeshSetupComplete (EventDispatchers|CreateEvent :!event Custom|MythicMaterialSetup)))

(event Custom|FinishProcess
  (params (Interacting%20Pawn :type Fort%20Pawn%20Object%20Reference))
  ;[0 0 1] Called by parent BP on interact
  (if (Variables|Default|GetCanInteract)
    (Networking|SwitchHasAuthority
      (:Authority
        (CallFunction|ServerGiveReward Interacting%20Pawn)
        (CallFunction|ServerSmashOpen Interacting%20Pawn)
        (CallFunction|ClearMarkers)))))

(event Interaction|EventBlueprintOnInteract
  (params (InteractingPawn :type Fort%20Pawn%20Object%20Reference) (InteractionBeingAttempted :type EInteractionBeingAttempted%20Enum :default "FirstInteraction"))
  ;[0 0 1] Called by parent BP on interact
  (Networking|SwitchHasAuthority
    (:Authority
      (Fortnite|ApplyGameplayEffectsToActor InteractingPawn (Utilities|Array|MakeArray "/HeistBasesGameplay/Items/Environmental/DisplayCase/GE_DisplayCase_GrantGA.GE_DisplayCase_GrantGA_C") (Transformation|GetActorLocation) false 0 (Variables|Default|GetAbilitySystem)))))

(event Custom|Server_DoLockdownBehavior
  ;[0.4 0.1 0] Lock cases down
  (Networking|SwitchHasAuthority
    (:Authority
      (CallFunction|ClearMarkers)
      (if (not (Variables|Default|GetLooted))
        (CallFunction|MulticastLockdown)))))

(event Custom|ExtendBarricade
  ;[0.4 0.1 0] Bring barricade up
  (Audio|PlaySoundatLocation "/HeistBasesGameplay/Audio/GlassCase/Heist_GlassCase_Barricade.Heist_GlassCase_Barricade" (Transformation|GetActorLocation))
  (bind (_direction _alpha) (Timeline|Timeline_0 PlayFromStart)
    (:Update
      (Transformation|SetRelativeLocation (Variables|Default|GetBarricade) (Math|Vector|Lerp(Vector) (Variables|Default|GetInitialBarricadeLocation) "0, 0, 0" _alpha) false false))))

(event Custom|Server_SmashOpen
  (params (InteractingPawn :type Fort%20Pawn%20Object%20Reference))
  (bind _self self)
  ;[0.4 0.1 0] Trigger lockdown for the base and call cosmetic event
  (bind _structout (Utilities|Struct|SetmembersinStructHeistBasesLockdownInitiatedPayload (Variables|Default|GetLockdownInitiated_Payload) InteractingPawn _self))
  (Messaging|BroadcastMessage (GameInstanceSubsystems|GetGameplayMessageRouter) (Variables|Default|GetRequestLockdownMessageChannel) _structout false _self) ; Broadcasts the message to trigger base lockdown
  (CallFunction|ServerSmashGCNandFeedback InteractingPawn))

(event Custom|RetractCase
  ;[0.4 0.1 0] Retract case into floor
  (Utilities|FlowControl|Sequence
    (:then_0
      (GameplayCue|AddGameplayCue self (Variables|Default|GetCaseCloseCueTag)) ; Looping GCN for case scraping as it retracts
    )
    (:then_1
      (bind (_direction _alpha) (Timeline|Timeline PlayFromStart)
        (:Update
          (Variables|Default|SetPedestalPositionRep (Math|Vector|Lerp(Vector) (Variables|Default|GetInitialPedestalLocation) (- (Variables|Default|GetInitialPedestalLocation) (Variables|Default|GetRetractOffset)) _alpha)))
        (:Finished
          (GameplayCue|RemoveGameplayCue self (Variables|Default|GetCaseCloseCueTag)))))))

(event Custom|Multicast_Lockdown
  ;[0.4 0.1 0] Lock cases down
  (if (Variables|Default|GetCanInteract)
    (Variables|Default|SetCanInteract false)
    (Variables|Default|SetNotify_ExtendBarricade true)
    (Utilities|FlowControl|Delay (Variables|Default|GetPauseBeforeCaseRetract))
    (CallFunction|RetractCase)))

(event EventOnRep_Looted
  (bind _glass (Variables|Default|GetGlass))
  ;[0.5 0 1] On loot, change mesh to broken version
  (|CallParentFunction)
  (if (Variables|Default|GetLooted)
    (Rendering|SetVisibility (Variables|Default|GetBG) false true)
    (Components|StaticMesh|SetStaticMesh _glass (Variables|Default|GetBrokenMesh))
    (Rendering|Material|SetOverlayMaterial _glass 0 false 0 :!class /Script/Engine.MeshComponent)
    (Rendering|Material|SetVectorParameterValue (Class|BTechArtGlobalLibrary|GetDynamicMaterialInstance (Variables|Default|GetPedestal)) "SFX_Emissive_Tint(Expand)_3" "(R=0.765625,G=0.730520,B=0.690387,A=1.000000)" :!class /Script/Engine.MaterialInstanceDynamic)))

(event Custom|Clear%20Markers
  ;[0.5 0 1] Clear markers, call client code to prevent new markers
  (MarkerComponent|DeactivateMarkersonActor self)
  (Variables|Default|SetMarkable false))

(event Custom|Server_SmashGCNandFeedback
  (params (Interacting%20Pawn :type Fort%20Pawn%20Object%20Reference))
  (bind _returnvalue (Fortnite|GetFortPlayerControllerfromActor Interacting%20Pawn))
  ;[0.5 0 1] Force feedback, camera shake, glass shatter GCN
  (Game|Feedback|ClientPlayForceFeedback _returnvalue "/Game/Weapons/ForceFeedback/FF_Generic_Explosion_Small_Near.FF_Generic_Explosion_Small_Near" "None" false false false) ; Rumble
  (Game|Feedback|ClientStartCameraShake :self _returnvalue :Shake "/Game/Effects/Fort_Effects/BluePrints/CameraShakes/B_Small_Explosion_CameraShake.B_Small_Explosion_CameraShake_C" :Scale 1.0 :PlaySpace "CameraLocal") ; Camera Shake
  (GameplayCue|ExecuteGameplayCueOnActor(Burst) self (Variables|Default|GetCaseShatterCueTag) (Ability|GameplayCue|MakeGameplayCueParameters :NormalizedMagnitude 1.0 :RawMagnitude 1.0 :Location (Transformation|GetWorldLocation (Variables|Default|GetTreasureLocation)) :Normal "0,0,0" :GameplayEffectLevel 1 :AbilityLevel 1 :TargetAttachComponent (Variables|Default|GetTreasureLocation) :bReplicateLocationWhenUsingMinimalRepProxy false)) ; Glass shatter burst GCN
)

(event EventOnLockdownReset
  ;[0 0 0] Debug Only
  (Variables|Default|SetLooted false)
  (Variables|Default|SetCanInteract true)
  (CallFunction|MulticastReset))

(event Custom|multicast%20reset
  ;[0 0 0] Debug Only
  (CallFunction|ClientMeshSetup)
  (Components|StaticMesh|SetStaticMesh (Variables|Default|GetPedestal) "/Game/Environments/Asteria/Props/EclipseVaults/SM_EclipseVault_WeaponCase_A_A/Meshes/SM_EclipseVault_WeaponCase_A_A.SM_EclipseVault_WeaponCase_A_A")
  (Transformation|SetWorldLocation (Variables|Default|GetPedestal) (Variables|Default|GetInitialPedestalLocation) false false)
  (Transformation|SetWorldLocation (Variables|Default|GetBarricade) (Variables|Default|GetInitialBarricadeLocation) false false))

(fn BlueprintCanInteract
  (params (InteractingPawn :type Fort%20Pawn%20Object%20Reference) (InteractionBeingAttempted :type EInteractionBeingAttempted%20Enum :default "FirstInteraction") (InteractionType :type TInteractionType%20Enum :default "IT_NoInteraction"))
  (return (and (|CallParentFunction InteractingPawn) (not (GameplayTags|HasAnyMatchingGameplayTags InteractingPawn (GameplayTags|MakeLiteralGameplayTagContainer "(GameplayTags=((TagName=\"Gameplay.Action.Attacking\"),(TagName=\"Gameplay.Action.Death\"),(TagName=\"Gameplay.Action.Disco\"),(TagName=\"Gameplay.Action.Drown\"),(TagName=\"Gameplay.Action.KnockedBack\"),(TagName=\"Gameplay.Action.Player.AbilityWeapon\"),(TagName=\"Gameplay.Action.Player.AimDownsights\"),(TagName=\"Gameplay.Action.Player.AirFlip\"),(TagName=\"Gameplay.Action.Player.Ascender\"),(TagName=\"Gameplay.Action.Player.CarryingPlayer\"),(TagName=\"Gameplay.Action.Player.Clamber\"),(TagName=\"Gameplay.Action.Player.Emote\"),(TagName=\"Gameplay.Action.Player.Fall\"),(TagName=\"Gameplay.Action.Player.Hurdle\"),(TagName=\"Gameplay.Action.Player.QuickHealingItem\"),(TagName=\"Gameplay.Action.Player.QuickHeal\"),(TagName=\"Gameplay.Action.Player.Riding\"),(TagName=\"Gameplay.Action.Player.ReviveOther\"),(TagName=\"Gameplay.Action.Player.TacticalSprint\"),(TagName=\"Gameplay.Action.Player.ThrowConsumable\"),(TagName=\"Gameplay.Action.Player.ThrowingPlayer\"),(TagName=\"Gameplay.Status.DBNO\"),(TagName=\"PlayerAugments.SelectionTool.Equip\"),(TagName=\"MovementMode.Passenger\"),(TagName=\"MovementMode.Driving\")))"))))))

(fn CheckForDisableHotfix
  ;[0.5 0.2 0] If instance-editiable FNM bool is ticked, AND our FNM hotfix value is active, make this case uninteractible and start it underground
  (if (and (Variables|Default|GetDisableForFNM) (ScalableFloat|GetValueasBool (Variables|Default|GetFNMHotfixActive)))
    (Variables|Default|SetCanInteract false)
    (Variables|Default|SetPedestalPositionRep (- (Transformation|GetWorldLocation (Variables|Default|GetPedestal)) (Variables|Default|GetRetractOffset)))))

(fn GetMarkerDetailsId
  (return (Utilities|String|StringToName (Utilities|String|Append "MarkerDetails." (Utilities|GetObjectName (Variables|Default|GetAssigned%20Treasure))))))

(fn GetMinDistanceFromInteraction
  (return 2e+02))

(fn GetOverrideItemInteractAbilityTag
  (return "(TagName=\"Gameplay.Action.Player.InteractSearch.DisplayCase\")"))

(fn GetTargetTagsForQuestEvent
  (return 0))

(fn OnLockdownInitiated
  (CallFunction|ServerDoLockdownBehavior)
  (return true))

(fn OnRep_BarricadePositionRep
  (Transformation|SetWorldLocation (Variables|Default|GetBarricade) (Variables|Default|GetBarricadePositionRep) false false))

(fn OnRep_Markable
  (MarkableActor|SetCanBeMarked (Variables|Default|GetMarkable)))

(fn OnRep_Notify_ExtendBarricade
  (CallFunction|ExtendBarricade))

(fn OnRep_Notify_RetractCase)

(fn OnRep_PedestalPositionRep
  (Transformation|SetWorldLocation (Variables|Default|GetPedestal) (Variables|Default|GetPedestalPositionRep) false false))

(fn SetMaterialOverlay
  (params (Mesh :type Mesh%20Component%20Object%20Reference))
  (bind _material_instance_dynamic (Class|BTechArtGlobalLibrary|GetDynamicMaterialInstance Mesh))
  (Rendering|Material|SetOverlayMaterial Mesh "/HeistBasesGameplay/Items/Environmental/DisplayCase/Materials/MI_Display_Pickup_Glow.MI_Display_Pickup_Glow" false 0 :!class /Script/Engine.MeshComponent)
  (bind _returnvalue (Class|FortCosmeticMaterialBlueprintLibrary|GetorCreateDynamicOverlayMaterial Mesh))
  (Rendering|Material|SetVectorParameterValue _returnvalue "Color" (Variables|Default|GetMythicColor) :!class /Script/Engine.MaterialInstanceDynamic)
  (Utilities|IsValid _material_instance_dynamic
    (:Is%20Valid
      (bind _returnvalue_2 (Rendering|Material|GetTextureParameterValue _material_instance_dynamic "Normals" :!class /Script/Engine.MaterialInstanceDynamic))
      (Rendering|Material|SetTextureParameterValue _returnvalue "Normals" _returnvalue_2))))

(fn ConstructionScript
  ;[0 1 0] Store initial posiitons of the pedestal and barricade, check out hotfix values
  (|CallParentFunction)
  (Variables|Default|SetInitialPedestalLocation (Transformation|GetWorldLocation (Variables|Default|GetPedestal)))
  (Variables|Default|SetInitialBarricadeLocation (Variables|/Script/Engine.SceneComponent|GetRelativeLocation (Variables|Default|GetBarricade)))
  (CallFunction|CheckforDisableHotfix))
