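# M_Display_Pickup_Glow is a translucent, lit material for display case items in Heist Bases, named as a pickup glow.
# It delegates its look to the shared M_Pickup_Blueprint_Function and breaks that function's attribute result into the material outputs.
# It is flagged for use on skeletal meshes and with mesh deformers.
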Material /HeistBasesGameplay/Items/Environmental/DisplayCase/Materials/M_Display_Pickup_Glow
graph {
 properties {
  bUsedWithMeshDeformer=true
  bUsedWithSkeletalMesh=true
  blendMode=BLEND_Translucent
  pixelDepthOffsetMode=PDOM_Legacy
  usedShadingModels=Default Lit
 }
 nodes [BreakMaterialAttributes!(Attr=MaterialFunctionCall!(Vector 3 Normal Sobel, Vertex Normal WS, World Space Normal, Bounding box G) {materialFunction=/Game/Effects/Fort_Effects/Gameplay/Pickups/Materials/M_Pickup_Blueprint_Function}.Result)]
}
